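; Looping gameplay cue for the katana dash charge, handling its VFX and its presentation on non-locally controlled players.
; On becoming relevant it sets those pawns' custom time dilation from falling or ground scalable values, restoring 1.0 on cease.
; Tick shows a sound indicator while their katana is held and, once pending, sets the linked anim layer's charge play rate to the inverse dilation.
; Loop start passes a particle component to the weapon's VFX interface.
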(event EventTick
  (params (DeltaSeconds :type Float%20%28single-precision%29))
  (Utilities|FlowControl|Sequence
    (:then_0
      (bind _playerref (Variables|Default|GetPlayerRef)
        (:then
          (if (GameplayTags|HasTag (Tags|GetWeaponDataGameplayTags (Weapon|GetCurrentWeapon _playerref)) "(TagName=\"Item.Weapon.Melee.Sword.Katana\")")
            (SoundIndicator|ShowSoundIndicator (Actor|GetComponentbyClass (Weapon|GetCurrentWeapon _playerref) "/Script/FortniteGame.FortSoundIndicatorComponent") (Transformation|GetActorLocation _playerref) 0 1.0 0)))
        (:else)))
    (:then_1
      (if (Variables|Default|GetbWaitingForValidAnimBP)
        (bind _playerref_2 (Variables|Default|GetPlayerRef)
          (:then
            (bind _asfort_katana_layer_anim_instance (Utilities|Casting|CastToFortKatanaLayerAnimInstance (Components|SkeletalMesh|AnimationBlueprintLinking|GetLinkedAnimLayerInstancebyClass (Variables|/Script/Engine.Character|GetMesh _playerref_2) "/Script/KatanaGameplayRuntime.FortKatanaLayerAnimInstance")))
            (Variables|/Script/KatanaGameplayRuntime.FortKatanaLayerAnimInstance|SetDashChargePlayRate (Math|Float|SafeDivide 1.0 (Variables|Default|GetActualDashChargeTimeDilation)) _asfort_katana_layer_anim_instance)
            (Variables|Default|SetbWaitingForValidAnimBP false))
          (:else))))))

(event EventOnLoopingStart(EitherVFX-preferthisversion)
  (params (MyTarget :type Actor%20Object%20Reference) (Parameters :type Gameplay%20Cue%20Parameters%20Structure%20%28by%20ref%29) (ParticleComponents :type Array%20of%20FXSystem%20Component%20Object%20References) (AudioComponents :type Array%20of%20Audio%20Component%20Object%20References))
  (bind _asfort_pawn (Utilities|Casting|CastToFortPawn MyTarget))
  (Class|BMeleeKatanaFXInterface|AnimNotifyVFXVisualSetup(Message) (Weapon|GetCurrentWeapon _asfort_pawn) (Utilities|Array|Get(acopy) ParticleComponents)))

(fn OnCeaseRelevant
  (params (MyTarget :type Actor%20Object%20Reference) (Parameters :type Gameplay%20Cue%20Parameters%20Structure%20%28by%20ref%29))
  (Utilities|FlowControl|Sequence
    (:then_0
      (bind _asfort_player_pawn (Utilities|Casting|CastToFortPlayerPawn MyTarget))
      (if (not (Pawn|IsLocallyControlled _asfort_player_pawn))
        (Variables|/Script/Engine.Actor|SetCustomTimeDilation 1.0 _asfort_player_pawn)))
    (:then_1
      (bind _returnvalue (|CallParentFunction MyTarget Parameters))
      (return _returnvalue))))

(fn ConstructionScript)

(fn OnBecomeRelevant
  (params (MyTarget :type Actor%20Object%20Reference) (Parameters :type Gameplay%20Cue%20Parameters%20Structure%20%28by%20ref%29))
  (Utilities|FlowControl|Sequence
    (:then_0
      (bind _asfort_player_pawn (Utilities|Casting|CastToFortPlayerPawn MyTarget))
      (if (not (Pawn|IsLocallyControlled _asfort_player_pawn))
        (bind _output_get (Variables|Default|SetPlayerRef _asfort_player_pawn))
        (bind _output_get_2 (Variables|Default|SetActualDashChargeTimeDilation (Math|Float|SelectFloat (ScalableFloat|GetValueatLevel (Variables|Default|GetFallingDashChargeTimeDilation)) (ScalableFloat|GetValueatLevel (Variables|Default|GetGroundDashChargeTimeDilation)) (GameplayTags|HasMatchingGameplayTag _output_get "(TagName=\"MovementMode.Falling\")"))))
        (Variables|/Script/Engine.Actor|SetCustomTimeDilation _output_get_2 _output_get)
        (Variables|Default|SetbWaitingForValidAnimBP true)))
    (:then_1
      (bind _returnvalue (|CallParentFunction MyTarget Parameters))
      (return _returnvalue))))
