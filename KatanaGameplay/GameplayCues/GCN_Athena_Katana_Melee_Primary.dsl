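; Gameplay cue notify for the katana's primary melee attack, supplying its audio and physics-field feedback.
; On execute, it casts the target to a Fort player pawn and, if holding the katana, shows a sound indicator and plays its tagged attack sound through SoundLibrary.
; It also spawns a radial force field at the pawn when the PhysicsField console variable is enabled, then calls the parent.
; OnActive is empty.
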(fn OnActive
  (params (MyTarget :type Actor%20Object%20Reference) (Parameters :type Gameplay%20Cue%20Parameters%20Structure%20%28by%20ref%29)))

(fn OnExecute
  (params (MyTarget :type Actor%20Object%20Reference) (Parameters :type Gameplay%20Cue%20Parameters%20Structure%20%28by%20ref%29))
  (bind _asfort_player_pawn (Utilities|Casting|CastToFortPlayerPawn MyTarget)
    (:then
      (bind _asb_melee_katana (Utilities|Casting|CastToB_Melee_Katana (Weapon|GetCurrentWeapon _asfort_player_pawn))
        (:then
          (SoundIndicator|ShowSoundIndicatorforDuration (Variables|/Script/FortniteGame.FortWeapon|GetSoundIndicatorComponent _asb_melee_katana) (Transformation|GetActorLocation _asfort_player_pawn) "/KatanaGameplay/Items/Katana/Sounds/Katana_Attack_Spin_Cue.Katana_Attack_Spin_Cue" 1.0 1.2 0)
          (SoundLibrary|PlaySoundSimple (WorldSubsystems|GetSoundLibrarySubsystem _asb_melee_katana) _asb_melee_katana (Utilities|Struct|MakeSimpleSoundSettings (Variables|Default|GetAttackAudioTag) true false true false 1.0 1.0 0.0 0 0 "None" (Variables|/Script/Engine.Actor|GetRootComponent _asfort_player_pawn) "KeepRelativeOffset" "0.000000,35.000000,0.000000") :!class /Script/SoundLibrary.SoundLibrarySubsystem))
        (:CastFailed
          _))
      (bind _returnvalue (Development|GetConsoleVariableIntValue "r.PhysicsField.EnableField"))
      (if (== _returnvalue 1)
        (Game|SpawnActorfromClass 0 "/Game/Effects/Niagara/Blueprints/PhsyicsFields/FN_RadialForce.FN_RadialForce_C" (Math|Transform|MakeTransform (Transformation|GetActorLocation _asfort_player_pawn)) "Undefined" "MultiplyWithRoot" 0 0.5 0.75 8e+02 0.7 0.0 0))
      (return (|CallParentFunction MyTarget Parameters)))
    (:CastFailed
      _)))
