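; Gameplay cue for the katana Mantis dash attack that configures its Niagara burst effect from the target pawn's current weapon.
; On burst, it passes the first spawned particle component to the weapon's melee katana FX interface for visual setup.
; It then reads the Creative melee weapon interface to set the block-break, custom color and intensity parameters, falling back to white and zero intensity without a custom color.
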(fn OnBurst(Niagara)
  (params (MyTarget :type Actor%20Object%20Reference) (Parameters :type Gameplay%20Cue%20Parameters%20Structure%20%28by%20ref%29) (ParticleComponents :type Array%20of%20Niagara%20Particle%20System%20Component%20Object%20References) (AudioComponents :type Array%20of%20Audio%20Component%20Object%20References) (BurstCameraShakeInstance :type Camera%20Shake%20Base%20Object%20Reference) (BurstDecalInstance :type Decal%20Actor%20Object%20Reference))
  (bind _output (Utilities|Array|Get(acopy) ParticleComponents))
  (bind _asfort_pawn (Utilities|Casting|CastToFortPawn MyTarget))
  (Class|BMeleeKatanaFXInterface|AnimNotifyVFXVisualSetup(Message) (Variables|/Script/FortniteGame.FortPawn|GetCurrentWeapon _asfort_pawn) _output)
  (bind _ascreative_melee_weapon_interface (Utilities|Casting|CastToCreative_Melee_Weapon_Interface (Variables|/Script/FortniteGame.FortPawn|GetCurrentWeapon _asfort_pawn)))
  (bind _enabled (CreativeMeleeWeaponInterface|GetSecondaryBreakBlockEnabled _ascreative_melee_weapon_interface))
  (Effects|Components|ParticleSystem|SetFloatParameter _output "Blockbreak" (Math|Conversions|ToFloat(Boolean) _enabled))
  (bind (_usecustomcolor _customcolor _customvfxintensity) (CreativeMeleeWeaponInterface|GetCustomVFXColor _ascreative_melee_weapon_interface))
  (Effects|Components|ParticleSystem|SetColorParameter _output "Color" (Math|Color|SelectColor _customcolor "(R=1.000000,G=1.000000,B=1.000000,A=1.000000)" _usecustomcolor))
  (Effects|Components|ParticleSystem|SetFloatParameter _output "UseCustomColor" (select _usecustomcolor 1.0 0.0))
  (Effects|Components|ParticleSystem|SetFloatParameter _output "CustomColorIntensity" (Math|Float|SelectFloat _customvfxintensity 0.0 _usecustomcolor))
  (return))
